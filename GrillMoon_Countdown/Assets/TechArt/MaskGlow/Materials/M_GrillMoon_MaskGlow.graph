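# Character material for glowing GrillMoon Countdown masks, built on the 2023 character parent functions.
# Layers cylindrically mapped panning fire noise, optional cracks, fresnel glow and overlay into emissive, added to scaled parent emissive.
# MPC_GrillMoon drives opacities, brightness and phase, which a per-primitive Mask Index offsets to stagger brightness pulses.
# World position offset samples panning noise at the component pivot, jittering each mask as a whole.
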Material /GrillMoon_Countdown/Assets/TechArt/MaskGlow/Materials/M_GrillMoon_MaskGlow
graph {
 properties {
  bAutomaticallySetUsageInEditor=false
  bUseMaterialAttributes=true
  bUsedWithMeshDeformer=true
  bUsedWithSkeletalMesh=true
  shadingModel=MSM_FromMaterialExpression
  usedShadingModels=Default Lit | Subsurface | Clear Coat | Subsurface Profile | Cloth
 }
 nodes [
  n1 = CollectionParameter() {collection=/GrillMoon_Countdown/Assets/FX/MPC_GrillMoon, parameterName=Mask_Phase}
  n2 = NamedRerouteDeclaration(None=ScalarParameter() {bUseCustomPrimitiveData=true, parameterName=Mask Index}) {name=Mask Index, nodeColor=<rgba 0 0.2266 1 1>}
  NamedRerouteDeclaration(None=Saturate(None=Subtract(A=@n1, B=@n2))) {name=Mask Phase, nodeColor=<rgba 0 0.0859 1 1>}
  n3 = Constant3Vector() {constant=<rgba 0 0 1 0>}
  n4 = ScalarParameter() {defaultValue=0.5, parameterName=WPO Noise Plane Rotation}
  n5 = Constant3Vector()
  n6 = MaterialFunctionCall() {materialFunction=&ComponentPivotLocation}
  n7 = Multiply(
   A=ComponentMask(None=RotateAboutAxis(NormalizedRotationAxis=@n3, RotationAngle=@n4, PivotPoint=@n5, Position=@n6.Result)) {b=true, r=true}
   B=ScalarParameter() {defaultValue=0.001, parameterName=WPO Noise Plane Tiling}
  )
  n8 = Time()
  n9 = MaterialFunctionCall(X=ScalarParameter() {defaultValue=0.1, parameterName=WPO Noise Plane Speed X}, Y=ScalarParameter() {parameterName=WPO Noise Plane Speed Y}) {materialFunction=&MakeFloat2}
  n10 = Subtract(A=TextureSample(UVs=Panner(Coordinate=@n7, Time=@n8, Speed=@n9.Result)) {samplerType=SAMPLERTYPE_LinearColor, texture=&HighContrastNoise}.RGB) {constB=0.5}
  n11 = ScalarParameter() {defaultValue=5, parameterName=WPO Noise Plane Offset Strength}
  n12 = CollectionParameter() {collection=&MPC_GrillMoon, parameterName=Mask_RandomWPO_Strength}
  NamedRerouteDeclaration(None=Multiply(A=Multiply(A=@n10, B=@n11), B=Saturate(None=@n12))) {name=WPO_Out, nodeColor=<rgba 0.4062 1 0 1>}
  n13 = VectorParameter() {parameterName=Noise Color, defaultValue=<rgba 1 0.8025 0.3021 1>}
  n14 = NamedRerouteUsage()
  n15 = VertexInterpolator(VS=LocalPosition() {includedOffsets=ExcludeOffsets, localOrigin=InstancePreSkinning}.XYZ)
  n16 = Normalize(VectorInput=ComponentMask(None=@n15.PS) {g=true, r=true})
  n17 = ComponentMask(None=@n16) {r=true}
  n18 = ComponentMask(None=@n16) {g=true}
  n19 = ConstantBiasScale(None=Divide(A=Arctangent2Fast(Y=@n17, X=@n18)) {constB=3.1416})
  n20 = Floor(None=Max(A=ScalarParameter() {defaultValue=2, parameterName=Noise Tiling XZ}))
  n21 = ComponentMask(None=@n15.PS) {b=true}
  n22 = Multiply(
   A=Add(A=@n21, B=ScalarParameter() {defaultValue=9, parameterName=Noise Offset Y})
   B=ScalarParameter() {defaultValue=0.2, parameterName=Noise Tiling Y}
  )
  n23 = ScalarParameter() {defaultValue=0.5, parameterName=Noise Power Y}
  n24 = NamedRerouteUsage()
  n25 = Panner(
   Coordinate=AppendVector(A=Multiply(A=@n19, B=@n20), B=Power(Base=@n22, Exp=@n23))
   Time=Add(A=Time(), B=Multiply(A=@n24) {constB=0.11})
   Speed=AppendVector(A=Constant(), B=ScalarParameter() {defaultValue=-0.5, parameterName=Noise Speed Y})
  )
  n26 = TextureObject() {samplerType=SAMPLERTYPE_Grayscale, texture=/Game/Effects/Fort_Effects/Textures/Fire/T_WispyFireNoise_Soft_M}
  n27 = Multiply(
   A=Add(A=@n21, B=ScalarParameter() {defaultValue=8, parameterName=Noise Falloff Gradient Offset})
   B=ScalarParameter() {defaultValue=0.075, parameterName=Noise Falloff Gradient Width}
  )
  n28 = Saturate(None=Power(Base=@n27, Exp=ScalarParameter() {defaultValue=3, parameterName=Noise Falloff Gradient Power 1}))
  n29 = Saturate(None=Subtract(A=TextureSample(UVs=@n25, Tex=@n26).R, B=@n28))
  n30 = Saturate(None=Power(Base=@n27, Exp=ScalarParameter() {defaultValue=0.1, parameterName=Noise Falloff Gradient Power 2}))
  n31 = Multiply(
   A=LinearInterpolate(B=Multiply(A=@n13.RGB, B=@n14), Alpha=LinearInterpolate(B=@n29, Alpha=@n30) {constA=1, constB=0})
   B=Saturate(None=CollectionParameter() {collection=&MPC_GrillMoon, parameterName=Mask_Noise_Opacity})
  )
  n32 = VectorParameter() {parameterName=Cracks Color, defaultValue=<rgba 1 0.8025 0.3021 1>}
  n33 = NamedRerouteUsage()
  n34 = TextureSample(Tex=TextureObjectParameter() {parameterName=Mask Cracks Mask, texture=/GrillMoon_Countdown/Assets/TechArt/MaskGlow/Textures/T_Glow_Mask}) {texture=&T_Glow_Mask}
  n35 = Saturate(None=CollectionParameter() {collection=&MPC_GrillMoon, parameterName=Mask_Cracks_Opacity})
  n36 = StaticSwitchParameter(True=LinearInterpolate(A=@n31, B=Multiply(A=@n32.RGB, B=@n33), Alpha=Multiply(A=@n34.R, B=@n35)), False=@n31) {parameterName=Use Cracks}
  n37 = Multiply(A=VectorParameter() {parameterName=Glow Color, defaultValue=<rgba 1 0.8025 0.3021 1>}.RGB, B=NamedRerouteUsage())
  n38 = PixelNormalWS()
  n39 = CameraVectorWS()
  n40 = Saturate(None=InverseLinearInterpolate(
    A=ScalarParameter() {defaultValue=0.3, parameterName=Glow Dot Min}
    B=ScalarParameter() {defaultValue=0.8, parameterName=Glow Dot Max}
    Value=DotProduct(A=@n38, B=@n39)
   ))
  n41 = Multiply(
   A=Power(Base=OneMinus(None=@n40), Exp=ScalarParameter() {defaultValue=2, parameterName=Glow Dot Power})
   B=Saturate(None=CollectionParameter() {collection=&MPC_GrillMoon, parameterName=Mask_Fresnel_Opacity})
  )
  n42 = NamedRerouteUsage()
  n43 = VectorParameter() {parameterName=Overlay Emissive Color, defaultValue=<rgba 1 0.8025 0.3021 1>}
  n44 = CollectionParameter() {collection=&MPC_GrillMoon, parameterName=Mask_Overlay_Emissive_Opacity}
  NamedRerouteDeclaration(None=LinearInterpolate(A=LinearInterpolate(A=@n36, B=@n37, Alpha=@n41), B=Multiply(A=@n42, B=@n43.RGB), Alpha=Saturate(None=@n44))) {name=Emissive_Out, nodeColor=<rgba 0.4062 1 0 1>}
  n45 = NamedRerouteUsage()
  n46 = Abs(None=ConstantBiasScale(None=@n45) {bias=-0.5, scale=2})
  n47 = CollectionParameter() {collection=&MPC_GrillMoon, parameterName=Mask_Brightness}
  n48 = ScalarParameter() {defaultValue=1.5, parameterName=Mask Pulse Brightness Multiplier}
  NamedRerouteDeclaration(None=MaterialFunctionCall(Input=OneMinus(None=@n46), Input Low=Constant(), Input High=Constant() {r=1}, Target Low=@n45, Target High=Multiply(A=@n47, B=@n48)) {materialFunction=&RemapValueRange}.Result) {name=Emissive Brightness, nodeColor=<rgba 1 0.1641 0 1>}
  OneMinus(None=@n28)
  n49 = MaterialFunctionCall() {materialFunction=&MF_CharacterParent_2023_PreFX}
  n50 = MaterialFunctionCall(Material=@n49.Result, M=@n49.Result, Pre Skinned Local Position=@n49.Result, Distance Blend=@n49.Result, Fresnel=@n49.Result) {materialFunction=&MF_CharacterParent_2023_PostFX}
  n51 = GetMaterialAttributes(None=@n50.Result) {attributeGetTypes [#1]}
  n52 = Multiply(A=NamedRerouteUsage(), B=ScalarParameter() {defaultValue=1, parameterName=Eye Emissive Strength})
  SetMaterialAttributes!(
   MaterialAttributes=@n50.Result
   World Position Offset=NamedRerouteUsage()
   Emissive Color=Add(A=Multiply(A=@n51.Emissive Color, B=@n52), B=NamedRerouteUsage())
  ) {attributeSetTypes [#2, #1]}
  VertexInterpolator(VS=LocalPosition() {localOrigin=InstancePreSkinning}.XYZ)
 ]
}
